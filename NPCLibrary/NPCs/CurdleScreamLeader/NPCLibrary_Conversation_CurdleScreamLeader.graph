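# Conversation database for the Curdle Scream Leader NPC, defining what it says when a player talks to it.
# A single entry point leads to one speech node with a default scream and contextual lines for a first-ever conversation or for specific player character skins.
# The speech node then branches through link nodes to the shared NPC quests and basic services conversations.
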ConversationDatabase /NPCLibrary/NPCs/CurdleScreamLeader/NPCLibrary_Conversation_CurdleScreamLeader
entry_points [{tag=Conversation.Entry.CurdleScreamLeader, destinations [#1]}]
nodes [
 ConversationEntryPointNode_7 {
  id=#1
  class=ConversationEntryPointNode
  outputs [#2]
  properties {entryTag {tagName=Conversation.Entry.CurdleScreamLeader}}
 }
 Speech_C_4 {
  id=#2
  class=Speech_C
  outputs [#3, #4]
  properties {
   nodeName=""
   generalConfig {
    defaultMessage="AAAAAAH! AARRRGGGGH!"
    contextualMessages [
     {
      message="AAAAH! ARRGH! AAAAAH!"
      contextRequirements [{requirement {class=FortControllerRequirement_IsFirstEverConversationWithNPC}}]
      priority {curve {rowName=Tandem.Speech.Priority.FirstEverInteraction}}
     }
     {
      message=Mom...my?
      requirementMatchPolicy=RequireAny
      contextRequirements [{requirement {
         class=FortControllerRequirement_HasCID
         properties {allowedCIds [
           /Game/Athena/Items/Cosmetics/Characters/CID_069_Athena_Commando_F_PinkBear
           /BRCosmetics/Athena/Items/Cosmetics/Characters/CID_781_Athena_Commando_F_FuzzyBearTEDDY
           /BRCosmetics/Athena/Items/Cosmetics/Characters/CID_109_Athena_Commando_M_Pizza
           /BRCosmetics/Athena/Items/Cosmetics/Characters/CID_232_Athena_Commando_F_HalloweenTomato
           /BRCosmetics/Athena/Items/Cosmetics/Characters/CID_258_Athena_Commando_F_FuzzyBearHalloween
           /BRCosmetics/Athena/Items/Cosmetics/Characters/CID_554_Athena_Commando_F_MilitiaMascotCuddle
           /BRCosmetics/Athena/Items/Cosmetics/Characters/CID_580_Athena_Commando_M_CuddleTeamDark
           /BRCosmetics/Athena/Items/Cosmetics/Characters/CID_916_Athena_Commando_F_FuzzyBearSkull
          ]}
        }}]
      priority {value=0, curve {curveTable=null, rowName=None}}
     }
     {
      message=(Curdle Scream Leader twitches in an unsettling manner.)
      contextRequirements [{requirement {class=FortControllerRequirement_HasCID, properties {allowedCIds [&Character_Bites]}}}]
      priority {value=0, curve {curveTable=null, rowName=None}}
     }
     {
      message=(Curdle Scream Leader unleashes a horrified yell.)
      contextRequirements [{requirement {class=FortControllerRequirement_HasCID, properties {allowedCIds [&Character_TheHerald]}}}]
      priority {value=0, curve {curveTable=null, rowName=None}}
     }
     {
      message=Erh? Errargh argh.
      contextRequirements [{requirement {class=FortControllerRequirement_HasCID, properties {allowedCIds [&Character_BadBear]}}}]
      priority {value=0, curve {curveTable=null, rowName=None}}
     }
     {
      message=(Curdle Scream Leader gives the Academy fan salute.)
      contextRequirements [{requirement {class=FortControllerRequirement_HasCID, properties {allowedCIds [&Character_PinkSpike]}}}]
      priority {value=0, curve {curveTable=null, rowName=None}}
     }
     {
      message=You're...my...fa...vor...ite.
      contextRequirements [{requirement {class=FortControllerRequirement_HasCID, properties {allowedCIds [&Character_RoseDust]}}}]
      priority {value=0, curve {curveTable=null, rowName=None}}
     }
     {
      message=Goooood...kit...ty.
      requirementMatchPolicy=RequireAny
      contextRequirements [{requirement {
         class=FortControllerRequirement_HasCID
         properties {allowedCIds [
           &CID_693_Athena_Commando_M_BuffCat
           /BRCosmetics/Athena/Items/Cosmetics/Characters/CID_776_Athena_Commando_M_ProfessorPup
           &Character_ChillCat
          ]}
        }}]
      priority {value=0, curve {curveTable=null, rowName=None}}
     }
    ]
   }
  }
 }
 ConversationLinkNode_0 {id=#4, class=ConversationLinkNode, properties {remoteEntryTag {tagName=Conversation.Entry.BasicServices}}}
 ConversationLinkNode_1 {id=#3, class=ConversationLinkNode, properties {remoteEntryTag {tagName=Conversation.Entry.NPCLibrary.Quests}}}
]
